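# Niagara ribbon trail for an animated swing, filed under the Ashton Chicago backpack cosmetic effects.
# Swing_Ribbon spawns along a falling rate curve over a brief emitter window, making short-lived pale blue segments.
# Two skeletal mesh samples, from User.SkeletalMesh and User.Pawn, supply only rotations, multiplied to set ribbon facing rather than position.
# Alpha and width shrink over particle life, while emitter age drives an initial brightness flash.
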NiagaraSystem /BRCosmetics/Athena/Items/Gameplay/BackPacks/Ashton/Chicago/FX/NS_CandyApple_AnimTrail
dependencies {usedDataInterfaces [&NiagaraDataInterfaceCurve], usedDynamicInputs [], usedModules [&SystemState]}
emitters {Swing_Ribbon {
  emitterUpdateScript [
   &EmitterState(Life Cycle Mode=::NewEnumerator1, Loop Behavior=::NewEnumerator1, Loop Duration=0.333, UseLoopDelay=false)
   &SpawnRate(
    SpawnRate=&FloatFromCurve(
     FloatCurve=DataInterface() {propertyValues {Curve {keys [{arriveTangent=-4.8837, interpMode=RCIM_Cubic, leaveTangent=-4.8837, tangentMode=RCTM_User, time=0, value=1}, <at 0.333 0 RCIM_Cubic>]}}}
     CurveIndex=$Emitter.Age
     Scale Curve=100
    )
    Use Spawn Probability=false
   )
  ]
  particleSpawnScript [
   &InitializeRibbon(
    Color=<rgba 0.5837 0.7919 1 0.3>
    Lifetime=0.22
    Module.Write Color=true
    Module.Write Lifetime=true
    Module.Write Mass=true
    Module.Write Material Random=true
    Module.Write Position=true
    Module.Write RibbonTwist=false
    Module.Write RibbonWidth=true
    Ribbon Width=40
   )
   &SkeletalMeshLocation(
    Apply To Particle Position=false
    Bone Sampling Mode=::NewEnumerator4
    Inherit Velocity=false
    MaskSpawnGroup=false
    Sampled Position Offset Coordinate Space:ENiagaraMeshTransforms=::NewEnumerator3
    Skeletal Mesh=$User.SkeletalMesh
   )
   &SkeletalMeshLocation(
    Apply To Particle Position=false
    Bone Sampling Mode=::NewEnumerator4
    Inherit Velocity=false
    MaskSpawnGroup=false
    Skeletal Mesh=$User.Pawn
   ) {module=SkeletalMeshLocation001}
   /SetVariables(Particles.RibbonFacing=&RotateVectorByQuaternion(
     Vector To Rotate=<0 1 0>
     Quaternion=&MultiplyQuaternion(Quaternion A=$Output.SkeletalMeshLocation001.SampledRotation, Quaternion B=$Output.SkeletalMeshLocation.SampledRotation)
    )) {isSetParameters=true}
  ]
  particleUpdateScript [
   &ParticleState()
   &SolveForcesAndVelocity(Clamp Velocity=false, Limit Acceleration=false)
   &ScaleColor(
    Scale Alpha=&FloatFromCurve(FloatCurve=DataInterface() {propertyValues {Curve {keys [<at 0 1>,<at 1 0>]}}})
    Scale RGB=&VectorFromFloat(Value=&FloatFromCurve(FloatCurve=DataInterface() {propertyValues {Curve {keys [<at 0 10>,<at 0.05 1>,<at 1 1>]}}}, CurveIndex=$Emitter.Age))
    ScaleA=true
    ScaleRGB=true
   )
   &ScaleRibbonWidth(Ribbon Width Scale=&FloatFromCurve(FloatCurve=DataInterface() {propertyValues {Curve {keys [
         <at 0 2 RCIM_Cubic>
         {arriveTangent=-1.5173, interpMode=RCIM_Cubic, leaveTangent=-1.5173, tangentMode=RCTM_User, time=0.1, value=1}
         <at 1 0.25>
        ]}}}))
  ]
  rendererClasses [&NiagaraRibbonRendererProperties]
  renderers [{rendererClass=&NiagaraRibbonRendererProperties}]
 }}
summary {
 systemName=NS_CandyApple_AnimTrail
 emitters [{emitterName=Swing_Ribbon, rendererClasses [&NiagaraRibbonRendererProperties]}]
 userVariables [
  User.Pawn {
   type=NiagaraDataInterfaceSkeletalMesh
   defaultValue {
    struct=NiagaraExt_VariableValue_DataInterface
    value {dataInterface=&self:NiagaraDataInterfaceSkeletalMesh_2}
   }
  }
  User.RibbonIDOffset {type=NiagaraInt32}
  User.SkeletalMesh {
   type=NiagaraDataInterfaceSkeletalMesh
   defaultValue {
    struct=NiagaraExt_VariableValue_DataInterface
    value {dataInterface=&self:NiagaraDataInterfaceSkeletalMesh_0}
   }
  }
 ]
}
